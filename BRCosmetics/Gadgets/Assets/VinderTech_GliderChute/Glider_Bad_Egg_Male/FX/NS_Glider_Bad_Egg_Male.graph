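# Niagara effect for the Bad Egg Male glider, drawing twin engine flames, smoke, point lights and ribbon trails.
# Emitters split particles into two spawn groups placed on the skeletal mesh, mirroring offsets or velocities by a per-group sign.
# Flames, smoke, lights and trails are gated on full deployment; smoke also needs an integrator threshold and is suppressed in front end.
# ForwardDot drives light brightness and trail spawning; system FlameInitialSize stretches engine flames.
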NiagaraSystem /BRCosmetics/Gadgets/Assets/VinderTech_GliderChute/Glider_Bad_Egg_Male/FX/NS_Glider_Bad_Egg_Male
dependencies {
 usedDataInterfaces [&NiagaraDataInterfaceColorCurve, &NiagaraDataInterfaceCurve, &NiagaraDataInterfaceVector2DCurve]
 usedDynamicInputs [&UniformAOrBFloat]
 usedModules [
  &SystemState
  &self:SystemSpawnScript.SystemScriptSource.SystemScriptGraph.NiagaraNodeAssignment_1.SetVariables_#1
  &FN_Integrator
  &self:SystemSpawnScript.SystemScriptSource.SystemScriptGraph.NiagaraNodeAssignment_0.SetVariables_#2
 ]
}
emitters {
 EngineFlames {
  emitterUpdateScript [
   &EmitterState()
   &SpawnRate(
    Spawn Probability=&MakeCustomFloatFromBool(Bool=$User.bIsFullyDeployed, True Float=1, False Float=0)
    SpawnRate=14
    Use Spawn Probability=true
   )
  ]
  particleSpawnScript [
   /Niagara/Modules/Spawn/Initialization/InitializeParticle(
    Lifetime=0.2
    Mesh Scale=&FN_FrontEndVector3Modifier(Gameplay=<0.325 0.325 0.5>, Front End=&Multiply_VectorByFloat(Vector=<0.325 0.325 0.5>, Float=0.75))
    Module.Write Color=true
    Module.Write Lifetime=true
    Module.Write Mass=false
    Module.Write Material Random=true
    Module.Write Position=false
    Module.Write Scale=true
    Module.Write SpriteRotation=false
    Module.Write SpriteSize=true
    Sprite Size=&Vector2DFromFloat(Value=&MakeCustomFloatFromBool(Bool=$User.bIsFrontEndPreview, True Float=10, False Float=22))
   )
   &FN_SetSpawnGroup(Number of Spawn Groups=2)
   /Game/Effects/Niagara/Modules/Math/FN_EvenNumberToSign(Input:NiagaraInt32=$Particles.SpawnGroup, Invert Sign=true, Output Float Attribute=true)
   &InitialMeshOrientation(
    Mesh Axis To Orient=&MaskVectorBySpawnGroup(Masked Vector=<0 0 1>, Passthrough Vector=<0 0 -1>)
    Mesh Orientation Mode=::NewEnumerator2
    Use Rotation Vector=false
   )
  ]
  particleUpdateScript [
   &ParticleState()
   /SetVariables(Particles.Offset:Vector3f=&MakeVector(
     X=&Multiply_Float(A=$Particles.Sign, B=&FN_FrontEndFloatModifier(Gameplay=12, Front End=6))
     Y=0
     Z=0
    )) {isSetParameters=true, module=SetVariables_#3}
   &SkeletalMeshLocation(
    Bone / Socket Index=$Particles.SpawnGroup
    Bone Sampling Mode=::NewEnumerator4
    MaskSpawnGroup=false
    Sampled Position Offset=&FN_FrontEndVector3Modifier(Gameplay=$Particles.Offset, Front End=&MakeVector(X=0, Y=0, Z=0))
    Skeletal Mesh=$User.SkeletalMesh
   )
   &ScaleMeshSize(Scale Factor=&MakeVector(X=1, Y=1, Z=&Add_Float(A=$System.FlameInitialSize, B=0.6)))
   &ScaleColor(
    Scale Alpha=&FloatFromCurve(FloatCurve=DataInterface() {propertyValues {Curve {keys [
         <at 0 1 RCIM_Cubic>
         {arriveTangent=-0.5882, interpMode=RCIM_Cubic, leaveTangent=-0.5882, time=0.6, value=0.95}
         {arriveTangent=-2.375, interpMode=RCIM_Cubic, leaveTangent=-2.375, time=0.85, value=0.5}
         <at 1 0 RCIM_Cubic>
        ]}}})
    Scale RGB=&MakeVectorFromLinearColorRGB(Color=&ColorFromCurve(ColorCurve=DataInterface() {propertyValues {
        AlphaCurve {keys []}
        BlueCurve {keys [<at 0 0 RCIM_Cubic>,<at 1 0 RCIM_Cubic>]}
        GreenCurve {keys [<at 0 1.1487 RCIM_Cubic>,<at 1 0.5063 RCIM_Cubic>]}
        RedCurve {keys [<at 0 2 RCIM_Cubic>,<at 1 1 RCIM_Cubic>]}
       }}))
    ScaleA=true
    ScaleRGB=false
   )
  ]
  rendererClasses [&NiagaraMeshRendererProperties]
  renderers [{rendererClass=&NiagaraMeshRendererProperties, rendererIndex=0}]
 }
 EngineFlamesCore {
  emitterUpdateScript [
   &EmitterState()
   &SpawnRate(
    Spawn Probability=&MakeCustomFloatFromBool(Bool=$User.bIsFullyDeployed, True Float=1, False Float=0)
    SpawnRate=50
    Use Spawn Probability=true
   )
  ]
  particleSpawnScript [
   &InitializeParticle(
    Lifetime=0.1
    Module.Write Color=true
    Module.Write Lifetime=true
    Module.Write Mass=false
    Module.Write Material Random=true
    Module.Write Position=false
    Module.Write Scale=false
    Module.Write SpriteRotation=true
    Module.Write SpriteSize=true
    Sprite Rotation=&UniformRangedFloat(Minimum=-360, Maximum=360)
    Sprite Size=&MakeCustomVector2DFromBool(Bool=$User.bIsFrontEndPreview, True Vector2D=<10 20>, False Vector2D=<14 28>)
   )
   &FN_SetSpawnGroup(Number of Spawn Groups=2)
   &FN_EvenNumberToSign(Input:NiagaraInt32=$Particles.SpawnGroup, Invert Sign=true, Output Float Attribute=true)
   /SetVariables(Particles.Offset:Vector3f=&MakeVector(X=0, Y=&FN_FrontEndFloatModifier(Gameplay=25, Front End=12), Z=0)) {isSetParameters=true, module=SetVariables_#3}
   &SkeletalMeshLocation(
    Bone / Socket Index=$Particles.SpawnGroup
    Bone Sampling Mode=::NewEnumerator4
    Interpolate Mesh Sampling=false
    MaskSpawnGroup=false
    Orientation Sampling=::NewEnumerator1
    Sampled Orientation Coordinate Space=::World
    Sampled Position Offset=$Particles.Offset
    Sampled Position Offset Coordinate Space:ENiagaraSkelMeshTransforms=::NewEnumerator4
    Skeletal Mesh=$User.SkeletalMesh
   )
   &AddVelocity(Velocity=<0 1 0>)
  ]
  particleUpdateScript [
   &ParticleState()
   &DynamicMaterialParameters(
    Index 0 Param 1=&FloatFromCurve(FloatCurve=DataInterface() {propertyValues {Curve {keys [<at 0.3 1 RCIM_Cubic>, {interpMode=RCIM_Cubic, leaveTangent=-8.3248, time=1, value=-1}]}}})
    Index 0 Param 4=1
    Module.Param0WriteEnabled=true
    Module.Param1WriteEnabled=true
    Module.Param2WriteEnabled=true
    Module.Param3WriteEnabled=true
   )
   &ScaleColor(
    Scale Alpha=&FloatFromCurve(FloatCurve=DataInterface() {propertyValues {Curve {keys [
         <at 0 1 RCIM_Cubic>
         {arriveTangent=-0.5882, interpMode=RCIM_Cubic, leaveTangent=-0.5882, time=0.6, value=0.95}
         {arriveTangent=-2.375, interpMode=RCIM_Cubic, leaveTangent=-2.375, time=0.85, value=0.5}
         <at 1 0 RCIM_Cubic>
        ]}}})
    Scale RGB=&MakeVectorFromLinearColorRGB(Color=&ColorFromCurve(ColorCurve=DataInterface() {propertyValues {
        AlphaCurve {keys [<at 0 1 RCIM_Cubic>,<at 0 1 RCIM_Cubic>]}
        BlueCurve {keys [<at 0 0.2083 RCIM_Cubic>,<at 0.998 0 RCIM_Cubic>]}
        GreenCurve {keys [<at 0 3.5069 RCIM_Cubic>,<at 0.998 0.4388 RCIM_Cubic>]}
        RedCurve {keys [<at 0 20 RCIM_Cubic>,<at 0.998 4 RCIM_Cubic>]}
       }}))
    ScaleA=true
    ScaleRGB=true
   )
   &SolveForcesAndVelocity(Clamp Velocity=false, Limit Acceleration=false)
   &ScaleSpriteSize(Scale Factor=&Vector2DFromCurve(
     Vector2Curve=DataInterface() {propertyValues {XCurve {keys [<at 0 1.5>,<at 1 0.25>]}, YCurve {keys [<at 0 1>,<at 1 4>]}}}
     ScaleCurve=&MakeVector2D(X=1, Y=&Add_Float(A=$System.FlameInitialSize, B=0.75))
    ))
  ]
  rendererClasses [&NiagaraSpriteRendererProperties]
  renderers [{rendererClass=&NiagaraSpriteRendererProperties}]
 }
 EngineSmoke {
  emitterUpdateScript [
   &EmitterState()
   &SpawnRate(
    Spawn Probability=&MakeCustomFloatFromBool(
     Bool=$User.bIsFullyDeployed
     True Float=&FN_FrontEndFloatModifier(
      Gameplay=&MakeCustomFloatFromBool(
       Bool=&SetBoolByFloatComparison(A=$System.FN_Integrator.NormalizedValue, B=0.25, Comparison Type=::NewEnumerator2)
       True Float=1
       False Float=0
      )
      Front End=0
     )
     False Float=0
    )
    SpawnRate=30
    Use Spawn Probability=true
   )
  ]
  particleSpawnScript [
   &FN_SetSpawnGroup(Number of Spawn Groups=2)
   &FN_EvenNumberToSign(Input:NiagaraInt32=$Particles.SpawnGroup, Output Float Attribute=true)
   &InitializeParticle(
    Color=<rgba 0.0885 0.0665 0.0604 1>
    Lifetime=&UniformRangedFloat(Minimum=1.5, Maximum=1.75)
    Mesh Scale=&VectorFromFloat(Value=&UniformRangedFloat(Minimum=0.035, Maximum=0.04))
    Module.Write Color=true
    Module.Write Lifetime=true
    Module.Write Mass=false
    Module.Write Material Random=true
    Module.Write Position=false
    Module.Write Scale=true
    Module.Write SpriteRotation=false
    Module.Write SpriteSize=false
   )
   &InitialMeshOrientation(
    Rotation=&MakeVector(X=&UniformRangedFloat(Minimum=-1, Maximum=1), Y=0, Z=0)
    Rotation Coordinate Space=::NewEnumerator0
    Use Rotation Vector=true
   )
   &SkeletalMeshLocation(
    Bone / Socket Index=$Particles.SpawnGroup
    Inherit Velocity=false
    Interpolate Mesh Sampling=false
    MaskSpawnGroup=false
    Mesh Sampling Type=::NewEnumerator5
    Sampled Orientation Coordinate Space=::World
    Sampled Position Offset=<0 25 -1>
    Sampled Position Offset Coordinate Space:ENiagaraMeshTransforms=::NewEnumerator2
    Skeletal Mesh=$User.SkeletalMesh
    Socket Sampling Mode=::NewEnumerator5
   )
   /SetVariables(Particles.TextureSize:NiagaraFloat=&UniformRangedFloat(Minimum=0.5, Maximum=1)) {isSetParameters=true}
   &AddVelocity(
    Coordinate Space=::World
    Velocity=&RotateVectorByQuaternion(
     Vector To Rotate=&MakeVector(
      X=&Multiply_Float(A=$Particles.Sign, B=&UniformRangedFloat(Minimum=325, Maximum=375))
      Y=&UniformRangedFloat(Minimum=-5, Maximum=5)
      Z=&UniformRangedFloat(Minimum=-5, Maximum=5)
     )
     Quaternion=$Output.SkeletalMeshLocation.SampledRotation
    )
   )
  ]
  particleUpdateScript [
   &DynamicMaterialParameters(
    Index 0 Param 1=$Particles.TextureSize
    Index 0 Param 2=&FloatFromCurve(FloatCurve=DataInterface() {propertyValues {Curve {keys [
         {arriveTangent=0.4856, interpMode=RCIM_Cubic, leaveTangent=0.4856, tangentMode=RCTM_User, time=-0.1, value=-0.5}
         {arriveTangent=-36.1175, interpMode=RCIM_Cubic, leaveTangent=-36.1175, tangentMode=RCTM_User, time=0.15, value=-4}
         {interpMode=RCIM_Cubic, leaveTangent=-27.9575, time=0.6, value=-12}
        ]}}})
    Index 0 Param 4=&FloatFromCurve(FloatCurve=DataInterface() {propertyValues {Curve {keys [<at 0 1 RCIM_Cubic>,<at 0.2 0.1 RCIM_Cubic>]}}})
    Module.Param0WriteEnabled=true
    Module.Param1WriteEnabled=true
    Module.Param2WriteEnabled=false
    Module.Param3WriteEnabled=true
   ) {module=DynamicMaterialParameters002}
   &ParticleState()
   &SolveForcesAndVelocity(Clamp Velocity=false, Limit Acceleration=false)
   &ScaleColor(
    Scale Alpha=&FloatFromCurve(FloatCurve=DataInterface() {propertyValues {Curve {keys [<at 0 0.89 RCIM_Cubic>,<at 0.9 0 RCIM_Cubic>]}}})
    Scale RGB=&MakeVectorFromLinearColorRGB(Color=&ColorFromCurve(ColorCurve=DataInterface() {propertyValues {
        AlphaCurve {keys []}
        BlueCurve {keys [<at 0 0 RCIM_Cubic>, <at 0.1 0>, {interpMode=RCIM_Cubic, leaveTangent=0.9084, time=0.2713, value=0.0257}]}
        GreenCurve {keys [<at 0 6.0284 RCIM_Cubic>, <at 0.1 0.0104>, {interpMode=RCIM_Cubic, leaveTangent=0.8989, time=0.2713, value=0.0265}]}
        RedCurve {keys [<at 0 30 RCIM_Cubic>, <at 0.1 0.1823>, {interpMode=RCIM_Cubic, leaveTangent=0.7428, time=0.2713, value=0.0295}]}
       }}))
    ScaleA=true
    ScaleRGB=true
   )
   &ScaleMeshSize(Scale Factor=&VectorFromFloat(Value=&FloatFromCurve(FloatCurve=DataInterface() {propertyValues {Curve {keys [<at 0 0.75>,<at 1 6>]}}})))
   &UpdateMeshOrientation()
  ]
  rendererClasses [&NiagaraMeshRendererProperties]
  renderers [{rendererClass=&NiagaraMeshRendererProperties, rendererIndex=0}]
 }
 FN_AnimTrail {
  emitterSpawnScript [&FN_RibbonIDOffsetUtility()]
  emitterUpdateScript [
   &EmitterState()
   &FN_SpawnAnimTrail(
    Enable:NiagaraBool=&Boolean_AndOperation(Bool1=$User.bIsFullyDeployed, Bool2=&SetBoolByFloatComparison(A=$User.ForwardDot, B=0.2))
    Number of Anim Trails=2
   )
  ]
  particleSpawnScript [
   &FN_InitializeAnimTrailBoneIndecies(First Bone or Socket Index=2, Second Bone or Socket Index=3)
   &FN_InitializeAnimTrailBoneIndecies(AnimTrail Selection=1, First Bone or Socket Index=5, Second Bone or Socket Index=4) {module=FN_InitializeAnimTrailBoneIndecies001}
   &FN_InitializeAnimTrail(Color=<rgba 1 1 1 0.25>, Lifetime=0.333)
   &AddVelocity(Velocity=<0 400 0>)
  ]
  particleUpdateScript [
   &SolveForcesAndVelocity(Clamp Velocity=false, Limit Acceleration=false)
   &FN_UpdateAnimTrail(Locked Particle RibbonWidth=1, SkeletalMesh=$User.SkeletalMesh) {module=FN_UpdateAnimTrail001}
   /SetVariables(Particles.RibbonU0Override=$Particles.NormalizedAge) {isSetParameters=true}
   &ScaleColor(
    Scale Alpha=&FloatFromCurve(FloatCurve=DataInterface() {propertyValues {Curve {keys [
         {arriveTangent=71.0767, interpMode=RCIM_Cubic, leaveTangent=71.0765, tangentMode=RCTM_User, time=0, value=0}
         <at 0.05 1 RCIM_Cubic>
         <at 0.8 0 RCIM_Cubic>
        ]}}})
    ScaleA=true
    ScaleRGB=true
   )
  ]
  rendererClasses [&NiagaraRibbonRendererProperties]
  renderers [{rendererClass=&NiagaraRibbonRendererProperties}]
 }
 FN_AttachedParticle_v3 {
  emitterUpdateScript [
   &EmitterState()
   &SpawnBurst_Instantaneous(
    Spawn Count=2
    Spawn Probability=HlslExpression() {hlslExpression="Emitter.LoopCount == 0"}
    Use Spawn Probability=false
    UseLoopCountLimit=true
   )
  ]
  particleSpawnScript [
   &FN_SetSpawnGroup(Number of Spawn Groups=2)
   &InitializeParticle(
    Module.Write Color=true
    Module.Write Lifetime=true
    Module.Write Mass=true
    Module.Write Material Random=true
    Module.Write Position=false
    Module.Write Scale=false
    Module.Write SpriteRotation=false
    Module.Write SpriteSize=false
    Position=&SimulationPosition()
   )
   &FN_EvenNumberToSign(Input:NiagaraInt32=$Particles.SpawnGroup, Invert Sign=true, Output Float Attribute=true)
  ]
  particleUpdateScript [
   &ParticleState(
    Kill Particles When Lifetime Has Elapsed=false
    Let Infinitely Lived Particles Die When Emitter Deactivates=true
    Loop Particles Lifetime=true
   )
   &SkeletalMeshLocation(
    Bone / Socket Index=$Particles.SpawnGroup
    Bone Sampling Mode=::NewEnumerator4
    MaskSpawnGroup=false
    Sampled Position Offset=&MakeVector(X=&Multiply_Float(A=$Particles.Sign, B=30), Y=0, Z=0)
    Skeletal Mesh=$User.SkeletalMesh
   )
   &FN_Light_Attributes(
    Color=<rgba 10 1.3744 0.0588 1>
    Create Color Attribute=true
    Create Exponent Attribute=true
    Create Radius Attribute=true
    Exponent=1
    Radius=50
    Scale Color=&Multiply_Float(
     A=&MakeCustomFloatFromBool(Bool=$User.bIsFullyDeployed, True Float=1, False Float=0)
     B=&Add_Float(
      A=&Lerp_Float(A=1, B=4, Alpha=&ClampFloat(Float=$User.ForwardDot))
      B=&Multiply_Float(
       A=&FloatFromCurve(FloatCurve=DataInterface() {propertyValues {Curve {keys [
            <at 0 0.5 RCIM_Cubic>
            {arriveTangent=-0.9405, interpMode=RCIM_Cubic, leaveTangent=-0.9405, time=0.1997, value=0.5694}
            {arriveTangent=2.3846, interpMode=RCIM_Cubic, leaveTangent=2.3846, time=0.2264, value=0.2871}
            {arriveTangent=-1.5552, interpMode=RCIM_Cubic, leaveTangent=-1.5552, time=0.3774, value=0.9932}
            {arriveTangent=-0.4954, interpMode=RCIM_Cubic, leaveTangent=-0.4954, time=0.4443, value=-0.0518}
            {arriveTangent=0.9255, interpMode=RCIM_Cubic, leaveTangent=0.9255, time=0.5069, value=0.929}
            {arriveTangent=-0.3061, interpMode=RCIM_Cubic, leaveTangent=-0.3061, time=0.6127, value=0.104}
            {arriveTangent=-0.3205, interpMode=RCIM_Cubic, leaveTangent=-0.3205, time=0.7465, value=0.8557}
            {arriveTangent=-1.4031, interpMode=RCIM_Cubic, leaveTangent=-1.4031, time=0.8415, value=0.0307}
            <at 1 0.5>
           ]}}})
       B=1
      )
     )
    )
   )
  ]
  rendererClasses [&NiagaraLightRendererProperties]
  renderers [{rendererClass=&NiagaraLightRendererProperties, rendererIndex=0}]
 }
 IdleFlames {
  emitterUpdateScript [
   &EmitterState()
   &SpawnRate(
    Spawn Probability=&MakeCustomFloatFromBool(Bool=$User.bIsFullyDeployed, True Float=1, False Float=0)
    SpawnRate=16
    Use Spawn Probability=true
   )
  ]
  particleSpawnScript [
   &InitializeParticle(
    Color=<rgba 1 1 1 0.5>
    Lifetime=2
    Module.Write Color=true
    Module.Write Lifetime=true
    Module.Write Mass=false
    Module.Write Material Random=true
    Module.Write Position=false
    Module.Write Scale=false
    Module.Write SpriteRotation=true
    Module.Write SpriteSize=true
    Sprite Rotation=&UniformRangedFloat(Minimum=-360, Maximum=360)
    Sprite Size=&Vector2DFromFloat(Value=&MakeCustomFloatFromBool(Bool=$User.bIsFrontEndPreview, True Float=10, False Float=22))
   )
   &FN_SetSpawnGroup(Number of Spawn Groups=2)
   &FN_EvenNumberToSign(Input:NiagaraInt32=$Particles.SpawnGroup, Output Float Attribute=true)
  ]
  particleUpdateScript [
   /SetVariables(Particles.Offset:Vector3f=&MakeVector(
     X=&Multiply_Float(A=$Particles.Sign, B=&FN_FrontEndFloatModifier(Gameplay=10, Front End=6))
     Y=0
     Z=0
    )) {isSetParameters=true, module=SetVariables_#3}
   &ParticleState()
   &SkeletalMeshLocation(
    Bone / Socket Index=$Particles.SpawnGroup
    Inherit Velocity=false
    Interpolate Mesh Sampling=false
    MaskSpawnGroup=false
    Mesh Sampling Type=::NewEnumerator5
    Sampled Position Offset=$Particles.Offset
    Sampled Position Offset Coordinate Space:ENiagaraMeshTransforms=::NewEnumerator3
    Skeletal Mesh=$User.SkeletalMesh
    Socket Sampling Mode=::NewEnumerator5
   )
   /SetVariables(Particles.SpriteFacing=&RotateVectorByQuaternion(
     Vector To Rotate=&MakeVector(X=$Particles.Sign, Y=0, Z=0)
     Quaternion=$Output.SkeletalMeshLocation.SampledRotation
    )) {isSetParameters=true}
   &DynamicMaterialParameters(
    Index 0 Param 1=&FloatFromCurve(FloatCurve=DataInterface() {propertyValues {Curve {keys [<at 0.3 1 RCIM_Cubic>, {interpMode=RCIM_Cubic, leaveTangent=-8.3248, time=1, value=-1}]}}})
    Index 0 Param 4=1
    Module.Param0WriteEnabled=true
    Module.Param1WriteEnabled=true
    Module.Param2WriteEnabled=true
    Module.Param3WriteEnabled=true
   )
   &ScaleColor(
    Scale Alpha=&FloatFromCurve(FloatCurve=DataInterface() {propertyValues {Curve {keys [<at 0 0 RCIM_Cubic>, <at 0.5 1 RCIM_Cubic>, {interpMode=RCIM_Cubic, leaveTangent=0.2, time=1, value=0}]}}})
    Scale RGB=&MakeVectorFromLinearColorRGB(Color=&ColorFromCurve(ColorCurve=DataInterface() {propertyValues {
        AlphaCurve {keys [<at 0 1 RCIM_Cubic>,<at 0 1 RCIM_Cubic>]}
        BlueCurve {keys [<at 0 0.4 RCIM_Cubic>,<at 0.998 0 RCIM_Cubic>]}
        GreenCurve {keys [<at 0 2 RCIM_Cubic>,<at 0.998 0.6582 RCIM_Cubic>]}
        RedCurve {keys [<at 0 10 RCIM_Cubic>,<at 0.998 6 RCIM_Cubic>]}
       }}))
    ScaleA=true
    ScaleRGB=true
   )
  ]
  rendererClasses [&NiagaraSpriteRendererProperties]
  renderers [{rendererClass=&NiagaraSpriteRendererProperties}]
 }
 SmokeCards {
  emitterSpawnScript [&FN_PreserveSkydiveUserVariables() {enabled=false, module=PreserveUserVariables}]
  emitterUpdateScript [
   &EmitterState()
   &SpawnRate(
    Spawn Probability=&MakeCustomFloatFromBool(
     Bool=$User.bIsFullyDeployed
     True Float=&FN_FrontEndFloatModifier(
      Gameplay=&MakeCustomFloatFromBool(
       Bool=&SetBoolByFloatComparison(A=$System.FN_Integrator.NormalizedValue, B=0.25, Comparison Type=::NewEnumerator2)
       True Float=1
       False Float=0
      )
      Front End=0
     )
     False Float=0
    )
    SpawnRate=34
    Use Spawn Probability=true
   )
  ]
  particleSpawnScript [
   &FN_SetSpawnGroup(Number of Spawn Groups=2)
   &FN_EvenNumberToSign(Input:NiagaraInt32=$Particles.SpawnGroup, Output Float Attribute=true)
   &InitializeParticle(
    Color=<rgba 0.0885 0.0665 0.0604 0.4>
    Lifetime=&UniformRangedFloat(Minimum=1, Maximum=1.25)
    Module.Write Color=true
    Module.Write Lifetime=true
    Module.Write Mass=false
    Module.Write Material Random=true
    Module.Write Position=false
    Module.Write Scale=false
    Module.Write SpriteRotation=true
    Module.Write SpriteSize=true
    Sprite Rotation=&UniformRangedFloat(Minimum=-360, Maximum=360)
    Sprite Size=&Vector2DFromFloat(Value=&UniformRangedFloat(Minimum=35, Maximum=45))
   )
   &SkeletalMeshLocation(
    Bone / Socket Index=$Particles.SpawnGroup
    Inherit Velocity=false
    Interpolate Mesh Sampling=false
    MaskSpawnGroup=false
    Mesh Sampling Type=::NewEnumerator5
    Sampled Orientation Coordinate Space=::World
    Sampled Position Offset=<0 35 0>
    Sampled Position Offset Coordinate Space:ENiagaraMeshTransforms=::NewEnumerator2
    Skeletal Mesh=$User.SkeletalMesh
    Socket Sampling Mode=::NewEnumerator5
   )
   &AddVelocity(
    Coordinate Space=::World
    Velocity=&RotateVectorByQuaternion(
     Vector To Rotate=&MakeVector(
      X=&Multiply_Float(A=$Particles.Sign, B=&UniformRangedFloat(Minimum=325, Maximum=375))
      Y=&UniformRangedFloat(Minimum=-5, Maximum=5)
      Z=&UniformRangedFloat(Minimum=-5, Maximum=5)
     )
     Quaternion=$Output.SkeletalMeshLocation.SampledRotation
    )
   )
   /SetVariables(Particles.RandomStartingFrame:NiagaraFloat=&UniformRangedFloat(Minimum=0, Maximum=63)) {isSetParameters=true}
  ]
  particleUpdateScript [
   &ParticleState()
   &ScaleColor(
    Scale Alpha=&FloatFromCurve(FloatCurve=DataInterface() {propertyValues {Curve {keys [<at 0 1 RCIM_Cubic>,<at 0.9 0 RCIM_Cubic>]}}})
    Scale RGB=&MakeVectorFromLinearColorRGB(Color=&ColorFromCurve(ColorCurve=DataInterface() {propertyValues {
        AlphaCurve {keys []}
        BlueCurve {keys [<at 0 0 RCIM_Cubic>, <at 0.15 0>, {interpMode=RCIM_Cubic, leaveTangent=0.9084, time=0.2713, value=0.0257}]}
        GreenCurve {keys [<at 0 3.0142 RCIM_Cubic>, <at 0.15 0.0104>, {interpMode=RCIM_Cubic, leaveTangent=0.8989, time=0.2713, value=0.0265}]}
        RedCurve {keys [<at 0 15 RCIM_Cubic>, <at 0.15 0.1823>, {interpMode=RCIM_Cubic, leaveTangent=0.7428, time=0.2713, value=0.0295}]}
       }}))
    ScaleA=true
    ScaleRGB=true
   )
   &SolveForcesAndVelocity(Clamp Velocity=false, Limit Acceleration=false)
   &ScaleSpriteSize(Scale Factor=&Vector2DFromFloat(Value=&FloatFromCurve(FloatCurve=DataInterface() {propertyValues {Curve {keys [<at 0 0.75 RCIM_Cubic>,<at 1 3>]}}})))
   /Niagara/Modules/Update/SubUV/SubUVAnimation(Number Of Frames=64, SubUV Lookup Index=&Add_Float(A=$Particles.NormalizedAge, B=$Particles.RandomStartingFrame))
   &DynamicMaterialParameters(
    Index 0 Param 1=&FloatFromCurve(FloatCurve=DataInterface() {propertyValues {Curve {keys [<at 0.3 1 RCIM_Cubic>,<at 1 -0.1>]}}})
    Index 0 Param 2=&FloatFromCurve(FloatCurve=DataInterface() {propertyValues {Curve {keys [<at 0 1 RCIM_Cubic>,<at 0.8 8.8 RCIM_Cubic>]}}})
    Index 0 Param 4=&FloatFromCurve(FloatCurve=DataInterface() {propertyValues {Curve {keys [<at 0.3 0>,<at 0.5 0.5>]}}})
    Module.Param0WriteEnabled=true
    Module.Param1WriteEnabled=true
    Module.Param2WriteEnabled=true
    Module.Param3WriteEnabled=true
   )
  ]
  rendererClasses [&NiagaraSpriteRendererProperties]
  renderers [{rendererClass=&NiagaraSpriteRendererProperties}]
 }
}
summary {
 systemName=NS_Glider_Bad_Egg_Male
 emitters [
  {emitterName=EngineFlames, rendererClasses [&NiagaraMeshRendererProperties]}
  {emitterName=EngineFlamesCore, rendererClasses [&NiagaraSpriteRendererProperties]}
  {emitterName=EngineSmoke, rendererClasses [&NiagaraMeshRendererProperties]}
  {emitterName=IdleFlames, rendererClasses [&NiagaraSpriteRendererProperties]}
  {emitterName=SmokeCards, rendererClasses [&NiagaraSpriteRendererProperties]}
  {emitterName=FN_AttachedParticle_v3, rendererClasses [&NiagaraLightRendererProperties]}
  {emitterName=FN_AnimTrail, rendererClasses [&NiagaraRibbonRendererProperties]}
 ]
 userVariables [
  User.bIsFrontEnd {type=NiagaraBool}
  User.bIsFrontEndPreview {type=NiagaraBool}
  User.bIsFullyDeployed {type=NiagaraBool}
  User.FlameInitialSize {type=NiagaraFloat, defaultValue=0, description=""}
  User.FlameSpawnProbability {type=NiagaraFloat, defaultValue=0, description=""}
  User.ForwardDot {type=NiagaraFloat, defaultValue=0, description=""}
  User.IsFullyDeployed {type=NiagaraFloat, defaultValue=0, description=""}
  User.LocalAlpha {type=NiagaraFloat, defaultValue=0, description=""}
  User.LocalSpawnProbability {type=NiagaraFloat, defaultValue=0, description=""}
  User.PlayerSpeed {type=NiagaraFloat, defaultValue=0, description=""}
  User.RibbonIDOffset {type=NiagaraInt32}
  User.SkeletalMesh {
   type=NiagaraDataInterfaceSkeletalMesh
   defaultValue {
    struct=NiagaraExt_VariableValue_DataInterface
    value {dataInterface=&self:NiagaraDataInterfaceSkeletalMesh_0}
   }
  }
 ]
}
